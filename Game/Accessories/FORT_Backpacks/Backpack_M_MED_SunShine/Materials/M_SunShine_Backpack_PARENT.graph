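# Parent material for the SunShine backpack, adding a time-based reactive emissive boost to the standard character look.
# Chains the 2023 character parent PreFX and PostFX functions, then overrides only Emissive Color through material attributes.
# The KillTime parameter minus current Time, clamped, multiplies emissive, so glow brightens while KillTime lies sufficiently ahead and ramps down linearly to normal as time catches up.
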Material /Game/Accessories/FORT_Backpacks/Backpack_M_MED_SunShine/Materials/M_SunShine_Backpack_PARENT
graph {
 properties {
  bUseMaterialAttributes=true
  bUsedWithMeshDeformer=true
  bUsedWithNanite=true
  bUsedWithSkeletalMesh=true
  pixelDepthOffsetMode=PDOM_Legacy
  shadingModel=MSM_FromMaterialExpression
  usedShadingModels=Default Lit | Subsurface | Clear Coat | Subsurface Profile | Cloth
 }
 nodes [
  n1 = MaterialFunctionCall() {materialFunction=&MF_CharacterParent_2023_PreFX}
  n2 = GetMaterialAttributes(None=MaterialFunctionCall(Material=@n1.Result, M=@n1.Result, Pre Skinned Local Position=@n1.Result, Distance Blend=@n1.Result, Fresnel=@n1.Result) {materialFunction=&MF_CharacterParent_2023_PostFX}.Result) {attributeGetTypes [#1]}
  n3 = Subtract(A=ScalarParameter() {group=Sunshine_Reactivity, parameterName=KillTime}, B=Time())
  SetMaterialAttributes(MaterialAttributes=@n2.MaterialAttributes, Emissive Color=Multiply(A=@n2.Emissive Color, B=Clamp(None=@n3) {maxDefault=10, minDefault=1})) {attributeSetTypes [#1]}
 ]
}
